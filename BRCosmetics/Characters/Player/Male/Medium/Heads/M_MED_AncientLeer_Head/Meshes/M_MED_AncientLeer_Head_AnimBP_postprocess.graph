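# Post-process Anim Blueprint for the male medium AncientLeer cosmetic head that adds secondary motion to its five leaf bones.
# The input pose passes through five serial single-bone Anim Dynamics nodes, Dyn_Leaf_05 down to 01, in component space.
# Each simulates in world space with damped angular springs, angular limits and overridden gravity, with Leaf_02 tuned softer and wider-swinging.
# The event graph holds only unconnected default nodes.
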AnimBlueprint /BRCosmetics/Characters/Player/Male/Medium/Heads/M_MED_AncientLeer_Head/Meshes/M_MED_AncientLeer_Head_AnimBP_postprocess
target_skeleton=&/M_MED_AncientLeer_Head_Skeleton
graphs [
 AnimGraph {nodes [
   n1 = Anim Dynamics - Bone: Dyn_Leaf_05(ComponentPose=Local To Component(LocalPose=Input Pose()), Alpha=1, ExternalForce="0,0,0") {
    angularDampingOverride=0.9
    angularSpringConstant=800
    bAngularSpring=true
    bGravityOverrideInSimSpace=true
    bOverrideAngularDamping=true
    bUseGravityOverride=true
    simulationSpace=World
    alphaBoolBlend {blendOption=Linear}
    boundBone {boneName=Dyn_Leaf_05}
    componentAppliedLinearAccClamp=<700 700 700>
    componentLinearAccScale=<0.65 0.64 1>
    componentLinearVelScale=<1 1 1>
    gravityOverride {z=-500}
    physicsBodyDefinitions [{
      boundBone {boneName=Dyn_Leaf_05}
      boxExtents=<10 10 10>
      constraintSetup {angularLimitsMax=<7 15 20>, angularLimitsMin=<-7 -15 -20>, angularTarget=<1 0 0>}
      localJointOffset=<1 0 0>
     }]
   }
   n2 = Anim Dynamics - Bone: Dyn_Leaf_04(ComponentPose=@n1, Alpha=1, ExternalForce="0,0,0") {
    angularDampingOverride=0.9
    angularSpringConstant=800
    bAngularSpring=true
    bGravityOverrideInSimSpace=true
    bOverrideAngularDamping=true
    bUseGravityOverride=true
    simulationSpace=World
    alphaBoolBlend {blendOption=Linear}
    boundBone {boneName=Dyn_Leaf_04}
    componentAppliedLinearAccClamp=<700 700 700>
    componentLinearAccScale=<0.65 0.64 1>
    componentLinearVelScale=<1 1 1>
    gravityOverride {z=-500}
    physicsBodyDefinitions [{
      boundBone {boneName=Dyn_Leaf_04}
      boxExtents=<15 15 15>
      constraintSetup {angularLimitsMax=<7 15 20>, angularLimitsMin=<-7 -15 -20>, angularTarget=<1 0 0>}
      localJointOffset=<0.8 0 0>
     }]
   }
   n3 = Anim Dynamics - Bone: Dyn_Leaf_03(ComponentPose=@n2, Alpha=1, ExternalForce="0,0,0") {
    angularDampingOverride=0.9
    angularSpringConstant=800
    bAngularSpring=true
    bGravityOverrideInSimSpace=true
    bOverrideAngularDamping=true
    bUseGravityOverride=true
    simulationSpace=World
    alphaBoolBlend {blendOption=Linear}
    boundBone {boneName=Dyn_Leaf_03}
    componentAppliedLinearAccClamp=<700 700 700>
    componentLinearAccScale=<0.65 0.64 1>
    componentLinearVelScale=<1 1 1>
    gravityOverride {z=-500}
    physicsBodyDefinitions [{
      boundBone {boneName=Dyn_Leaf_03}
      boxExtents=<10 10 10>
      constraintSetup {angularLimitsMax=<7 15 20>, angularLimitsMin=<-7 -15 -20>, angularTarget=<1 0 0>}
      localJointOffset=<1 0 0>
     }]
   }
   n4 = Anim Dynamics - Bone: Dyn_Leaf_02(ComponentPose=@n3, Alpha=1, ExternalForce="0,0,0") {
    angularDampingOverride=0.8
    angularSpringConstant=750
    bAngularSpring=true
    bGravityOverrideInSimSpace=true
    bOverrideAngularDamping=true
    bUseGravityOverride=true
    simulationSpace=World
    alphaBoolBlend {blendOption=Linear}
    boundBone {boneName=Dyn_Leaf_02}
    componentAppliedLinearAccClamp=<500 500 500>
    componentLinearAccScale=<0.5 0.5 0.8>
    componentLinearVelScale=<0.9 0.9 0.9>
    gravityOverride {z=-500}
    physicsBodyDefinitions [{
      boundBone {boneName=Dyn_Leaf_02}
      boxExtents=<12 12 12>
      constraintSetup {angularLimitsMax=<10 35 20>, angularLimitsMin=<-10 -35 -20>, angularTarget=<1 0 0>}
      localJointOffset=<1 0 0>
     }]
   }
   n5 = Anim Dynamics - Bone: Dyn_Leaf_01(ComponentPose=@n4, Alpha=1, ExternalForce="0,0,0") {
    angularDampingOverride=0.9
    angularSpringConstant=800
    bAngularSpring=true
    bGravityOverrideInSimSpace=true
    bOverrideAngularDamping=true
    bUseGravityOverride=true
    simulationSpace=World
    alphaBoolBlend {blendOption=Linear}
    boundBone {boneName=Dyn_Leaf_01}
    componentAppliedLinearAccClamp=<700 700 700>
    componentLinearAccScale=<0.65 0.64 1>
    componentLinearVelScale=<1 1 1>
    gravityOverride {z=-500}
    physicsBodyDefinitions [{
      boundBone {boneName=Dyn_Leaf_01}
      boxExtents=<9 9 9>
      constraintSetup {angularLimitsMax=<7 15 20>, angularLimitsMin=<-7 -15 -20>, angularTarget=<1 0 0>}
      localJointOffset=<0.9 0 0>
     }]
   }
   Output Pose(Result=Component To Local(ComponentPose=@n5))
  ]}
 EventGraph {nodes [
   TryGetPawnOwner()
   Event BlueprintUpdateAnimation()
  ]}
]
variables []
